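# Quest debug tool widget that shows a "Bundle Entries" section listing the quest entries of a bundle.
# A section header sits above a vertical Dynamic Entry Box inside a vertical box.
# Set Entries takes an array of bundle quest entry debug structs, resets the box, creates one QuestDebugToolBundleEntry widget per element and passes each its struct through Set Data.
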WidgetBlueprint /QuestUI/DebugTool/BundleWidgets/QuestDebugToolBundleEntryInfo
graphs [
 EventGraph {links [], nodes []}
 Set Entries {nodes [
   n1 = Set Entries!() -> (then, Entries:Array of Bundle Quest Entry Debug Data Structures) {is=title K2Node_FunctionEntry}
   n2 = Get DynamicEntryBox_Entries() -> (DynamicEntryBox_Entries:Dynamic Entry Box Object Reference)
   > Reset!(execute, self:Dynamic Entry Box Object Reference=@n2, bDeleteWidgets:bool=false) -> (then)
   > n3 = For Each Loop(Array:Array of Bundle Quest Entry Debug Data Structures=@n1) -> (Array Element:Bundle Quest Entry Debug Data Structure)
   > n4 = Create Entry!(execute, self:Dynamic Entry Box Object Reference=@n2) -> (then, ReturnValue:User Widget Object Reference)
   > n5 = Cast To QuestDebugToolBundleEntry!(execute, Object=@n4) -> (then, CastFailed, AsQuest Debug Tool Bundle Entry:Quest Debug Tool Bundle Entry Object Reference) {is=title K2Node_DynamicCast}
   > Set Data!(execute, self:Quest Debug Tool Bundle Entry Object Reference=@n5, Entry:"Bundle Quest Entry Debug Data Structure (by ref)"=@n3) -> (then) {is=title K2Node_CallFunction}
  ]}
]
variables []
widgets {
 info {inheritedWidgetCount=0, parentClass=&UserWidget, rootWidgetClass=&VerticalBox, widgetCount=3}
 widgets [
  VerticalBox_28 {widgetClassPath=&VerticalBox, widget=VerticalBox_28}
  SectionHeader {
   widgetClassPath=/QuestUI/DebugTool/CommonWidgets/QuestDebugToolSectionHeader
   bIsVariable=true
   parent=VerticalBox_28
   widget=SectionHeader
   properties {header=Bundle Entries}
   slot {class=VerticalBoxSlot}
  }
  DynamicEntryBox_Entries {
   widgetClassPath=&DynamicEntryBox
   bIsVariable=true
   parent=VerticalBox_28
   widget=DynamicEntryBox_Entries
   properties {
    entryBoxType=Vertical
    entryWidgetClass=&/QuestDebugToolBundleEntry.QuestDebugToolBundleEntry_C
    entrySpacing=<0 2>
   }
   slot {class=VerticalBoxSlot, properties {padding=<ltrb 6 0 6 6>}}
  }
 ]
}
