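# Quest debug tool panel that shows a quest bundle's definition, progress, state and calendar event timing.
# It stacks shared header, definition and info-field widgets, including an alert for an inactive calendar event.
# Helper functions format progress (N/A without a positive requirement), state and event name from their inputs, and derive visibility, warning only when a named event is not found.
# The event graph is empty, and no variables are declared.
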WidgetBlueprint /QuestUI/DebugTool/BundleWidgets/QuestDebugToolBundleInfo
graphs [
 EventGraph {links [], nodes []}
 Get Bundle Progress {nodes [
   n1 = Get Bundle Progress!() -> (then, Achieved:int, Required:int) {is=title K2Node_FunctionEntry}
   > n2 = Branch(Condition="integer > integer"(A=@n1.Required))
   > Return Node!(execute, Name:Text=Format Text!(Format="LOCTEXT(\"{0} / {1}\")", 0:int=@n1, 1:int=@n1.Required) -> (Result))
   Return Node(execute=@n2.else, Name="LOCTEXT(\"N/A\")")
  ]}
 Get Bundle State {nodes [
   n1 = Get Bundle State!() -> (then, BundleState:EFortQuestState Enum=Inactive) {is=title K2Node_FunctionEntry}
   > Return Node!(execute, State:Text="To Text (String)"(InString=Enum to String!(Enumerator:EFortQuestState Enum=@n1) -> (ReturnValue:String)))
  ]}
 Get Event Name Visibility {nodes [
   n1 = Get Event Name Visibility!() -> (then, BundleListItemId:String) {is=title K2Node_FunctionEntry}
   n2 = IsEmpty(InString=@n1)
   > Return Node!(execute, Visibility:ESlateVisibility Enum=Select(Option 0=Collapsed, Option 1=HitTestInvisible, Index=NOT Boolean(A=@n2)))
  ]}
 Get Event Timer Visibility {nodes [
   n1 = Get Event Timer Visibility!() -> (then, EventFound:bool=~) {is=title K2Node_FunctionEntry}
   > Return Node(Visibility=Select(Option 0=Collapsed, Option 1=HitTestInvisible, Index=@n1))
  ]}
 Get Event Warning Visibility {nodes [
   n1 = Get Event Warning Visibility!() -> (then, EventFound:bool=~, EventName:String) {is=title K2Node_FunctionEntry}
   n2 = IsEmpty(InString=@n1.EventName)
   > Return Node(Visibility=Select(Option 0=HitTestInvisible, Option 1=Collapsed, Index=OR Boolean(A=@n1, B=@n2)))
  ]}
 GetEventName {nodes [
   n1 = GetEventName!() -> (then, EventName:String) {is=title K2Node_FunctionEntry}
   > Return Node!(execute, OutEventName:Text=SelectText(A="LOCTEXT(\"(None)\")", B="To Text (String)"(InString=@n1), bPickA=IsEmpty(InString=@n1)))
  ]}
]
variables []
widgets {
 info {inheritedWidgetCount=0, parentClass=/Script/QuestUI.QuestPanelDebugToolWidgetBase, rootWidgetClass=&VerticalBox, widgetCount=11}
 widgets [
  VerticalBox_1 {widgetClassPath=&VerticalBox, widget=VerticalBox_1}
  SectionHeader {
   widgetClassPath=/QuestUI/DebugTool/CommonWidgets/QuestDebugToolSectionHeader
   bIsVariable=true
   parent=VerticalBox_1
   widget=SectionHeader
   properties {header=Bundle Info}
   slot {class=VerticalBoxSlot}
  }
  Widget_BundleDefinition {
   widgetClassPath=/QuestUI/DebugTool/CommonWidgets/QuestDebugToolDefinitionField
   bIsVariable=true
   parent=VerticalBox_1
   widget=Widget_BundleDefinition
   properties {assetEntryType=BundleAsset, definitionHeader=Bundle Definition, displayDataHeader=Bundle Display Data, viewEntryType=BundleView}
   slot {class=VerticalBoxSlot}
  }
  HorizontalBox_0 {
   widgetClassPath=&HorizontalBox
   parent=VerticalBox_1
   widget=HorizontalBox_0
   slot {class=VerticalBoxSlot, properties {size {sizeRule=Fill, value=1}}}
  }
  Widget_BundleProgress {
   widgetClassPath=/QuestUI/DebugTool/CommonWidgets/QuestDebugToolInfoField
   bIsVariable=true
   parent=HorizontalBox_0
   widget=Widget_BundleProgress
   properties {header=Progress}
   slot {class=HorizontalBoxSlot, properties {size {sizeRule=Fill, value=1}}}
  }
  Widget_BundleState {
   widgetClassPath=&QuestDebugToolInfoField
   bIsVariable=true
   parent=HorizontalBox_0
   widget=Widget_BundleState
   properties {header=State}
   slot {class=HorizontalBoxSlot, properties {size {sizeRule=Fill, value=1}}}
  }
  Widget_CalendarEvent {
   widgetClassPath=&QuestDebugToolInfoField
   bIsVariable=true
   parent=VerticalBox_1
   widget=Widget_CalendarEvent
   properties {header=Calendar Event}
   slot {class=VerticalBoxSlot}
  }
  Widget_EventWarning {
   widgetClassPath=&QuestDebugToolInfoField
   bIsVariable=true
   parent=VerticalBox_1
   widget=Widget_EventWarning
   properties {data=Calendar event is not currently active, displayAsAlert=true, header=Alert}
   slot {class=VerticalBoxSlot}
  }
  HorizontalBox_1 {widgetClassPath=&HorizontalBox, parent=VerticalBox_1, widget=HorizontalBox_1, slot {class=VerticalBoxSlot}}
  Widget_EventStart {
   widgetClassPath=&QuestDebugToolInfoField
   bIsVariable=true
   parent=HorizontalBox_1
   widget=Widget_EventStart
   properties {header=Event Start}
   slot {class=HorizontalBoxSlot, properties {size {sizeRule=Fill, value=1}}}
  }
  Widget_EventEnd {
   widgetClassPath=&QuestDebugToolInfoField
   bIsVariable=true
   parent=HorizontalBox_1
   widget=Widget_EventEnd
   properties {header=Event End}
   slot {class=HorizontalBoxSlot, properties {size {sizeRule=Fill, value=1}}}
  }
 ]
}
