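# Quest debug tool widget that shows punch card info for a quest bundle: a section header, a display style field and a header field.
# GetDisplayType returns "Loose Quests" when SkipAddToGoalBundles is set, otherwise the goal card orientation name.
# GetHeaderText switches on orientation, passing source text through only for Horizontal_Nonsequential and returning explanatory placeholders otherwise.
# The event graph is empty; the parent class is QuestPanelDebugToolWidgetBase.
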WidgetBlueprint /QuestUI/DebugTool/BundleWidgets/QuestDebugToolPunchCardInfo
graphs [
 EventGraph {links [], nodes []}
 GetDisplayType {nodes [
   n1 = GetDisplayType!() -> (then, SkipAddToGoalBundles:bool=~, Orientation:EGoalCardOrientation Enum=Vertical) {is=title K2Node_FunctionEntry}
   > n2 = Branch(Condition=@n1)
   > Return Node!(execute, DisplayType:Text="LOCTEXT(\"Loose Quests\")")
   Return Node(
    execute=@n2.else
    DisplayType="To Text (String)"(InString=Enum to String!(Enumerator:EGoalCardOrientation Enum=@n1.Orientation) -> (ReturnValue:String))
   )
  ]}
 GetHeaderText {nodes [
   n1 = GetHeaderText!() -> (then, SourceText:Text, Orientation:EGoalCardOrientation Enum=Vertical) {is=title K2Node_FunctionEntry}
   > n2 = Switch on EGoalCardOrientation!(execute, Selection:EGoalCardOrientation Enum=@n1.Orientation) -> (Vertical:Exec, Horizontal:Exec, Horizontal_Nonsequential:Exec, Daily:Exec) {is=title K2Node_SwitchEnum}
   Return Node!(execute=@n2.Daily, HeaderText:Text="LOCTEXT(\"(not used for Daily)\")")
   Return Node(execute=@n2.Horizontal, HeaderText="LOCTEXT(\"(first incomplete quest in bundle)\")")
   Return Node(execute=@n2.Horizontal_Nonsequential, HeaderText=@n1)
   > Return Node(HeaderText="LOCTEXT(\"(not used for Vertical)\")")
  ]}
]
variables []
widgets {
 info {inheritedWidgetCount=0, parentClass=/Script/QuestUI.QuestPanelDebugToolWidgetBase, rootWidgetClass=&VerticalBox, widgetCount=5}
 widgets [
  VerticalBox_1 {widgetClassPath=&VerticalBox, widget=VerticalBox_1}
  SectionHeader {
   widgetClassPath=/QuestUI/DebugTool/CommonWidgets/QuestDebugToolSectionHeader
   bIsVariable=true
   parent=VerticalBox_1
   widget=SectionHeader
   properties {header=Punch Card Info}
   slot {class=VerticalBoxSlot}
  }
  Widget_DisplayStyle {
   widgetClassPath=/QuestUI/DebugTool/CommonWidgets/QuestDebugToolInfoField
   bIsVariable=true
   parent=VerticalBox_1
   widget=Widget_DisplayStyle
   properties {header=Display Style}
   slot {class=VerticalBoxSlot}
  }
  Panel_WithPunchCard {
   widgetClassPath=&VerticalBox
   bIsVariable=true
   parent=VerticalBox_1
   widget=Panel_WithPunchCard
   slot {class=VerticalBoxSlot, properties {size {sizeRule=Fill, value=1}}}
  }
  Widget_Header {
   widgetClassPath=&QuestDebugToolInfoField
   bIsVariable=true
   parent=Panel_WithPunchCard
   widget=Widget_Header
   slot {class=VerticalBoxSlot, properties {size {sizeRule=Fill, value=1}}}
  }
 ]
}
