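# Quest debug tool widget that presents a "Build Info" section on plugin and asset versions.
# Info fields show plugin name, this version, plugin and asset release and sunset versions, plus an alert-styled version warning.
# Helper functions derive visibility from plugin residency and warning text, and choose the plugin name or fallback messages for missing build data or non-plugin content.
# The event graph is empty, and the helpers work only from passed-in values.
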WidgetBlueprint /QuestUI/DebugTool/CommonWidgets/QuestDebugToolPluginInfo
graphs [
 EventGraph {links [], nodes []}
 GetBuildWarningVisibility {nodes [
   n1 = GetBuildWarningVisibility!() -> (then, WarningText:Text) {is=title K2Node_FunctionEntry}
   > Return Node!(execute, Visibility:ESlateVisibility Enum=Select(Option 0=HitTestInvisible, Option 1=Collapsed, Index=TextIsEmpty(InText=@n1)))
  ]}
 GetPluginName {nodes [
   n1 = GetPluginName!() -> (then, PluginName:String, HasBuildData:bool=~) {is=title K2Node_FunctionEntry}
   > n2 = Branch(Condition=@n1.HasBuildData)
   > Return Node!(
    execute
    OutPluginName:Text=Select(Option 0:Text="To Text (String)"(InString=@n1), Option 1:Text="LOCTEXT(\"This content is not in a plugin\")", Index=IsEmpty(InString=@n1)) -> (ReturnValue:Text)
   )
   Return Node(execute=@n2.else, OutPluginName="LOCTEXT(\"Plugin data is unknown without a definition\")")
  ]}
 GetPluginVersionVisibility {nodes [
   n1 = GetPluginVersionVisibility!() -> (then, ResidesInPlugin:bool=~) {is=title K2Node_FunctionEntry}
   > Return Node!(
    execute
    Visibility:ESlateVisibility Enum=Select(Option 0:ESlateVisibility Enum=Collapsed, Option 1:ESlateVisibility Enum=HitTestInvisible, Index=@n1) -> (ReturnValue:ESlateVisibility Enum)
   )
  ]}
]
variables []
widgets {
 info {inheritedWidgetCount=0, parentClass=&UserWidget, rootWidgetClass=&VerticalBox, widgetCount=11}
 widgets [
  VerticalBox_79 {widgetClassPath=&VerticalBox, widget=VerticalBox_79}
  SectionHeader {
   widgetClassPath=&/QuestDebugToolSectionHeader
   bIsVariable=true
   parent=VerticalBox_79
   widget=SectionHeader
   properties {header=Build Info}
   slot {class=VerticalBoxSlot}
  }
  HorizontalBox_0 {widgetClassPath=&HorizontalBox, parent=VerticalBox_79, widget=HorizontalBox_0, slot {class=VerticalBoxSlot}}
  PluginName {
   widgetClassPath=&/QuestDebugToolInfoField
   bIsVariable=true
   parent=HorizontalBox_0
   widget=PluginName
   properties {header=Plugin Name}
   slot {class=HorizontalBoxSlot, properties {size {sizeRule=Fill, value=1}}}
  }
  PluginReleaseVersion {
   widgetClassPath=&QuestDebugToolInfoField
   bIsVariable=true
   parent=HorizontalBox_0
   widget=PluginReleaseVersion
   properties {header=Plugin Release}
   slot {class=HorizontalBoxSlot}
  }
  PluginSunsetVersion {
   widgetClassPath=&QuestDebugToolInfoField
   bIsVariable=true
   parent=HorizontalBox_0
   widget=PluginSunsetVersion
   properties {header=Plugin Sunset}
   slot {class=HorizontalBoxSlot}
  }
  HorizontalBox_163 {widgetClassPath=&HorizontalBox, parent=VerticalBox_79, widget=HorizontalBox_163, slot {class=VerticalBoxSlot}}
  BuildVersion {
   widgetClassPath=&QuestDebugToolInfoField
   bIsVariable=true
   parent=HorizontalBox_163
   widget=BuildVersion
   properties {header=This Version}
   slot {class=HorizontalBoxSlot, properties {size {sizeRule=Fill, value=1}}}
  }
  AssetReleaseVersion {
   widgetClassPath=&QuestDebugToolInfoField
   bIsVariable=true
   parent=HorizontalBox_163
   widget=AssetReleaseVersion
   properties {header=Asset Release}
   slot {class=HorizontalBoxSlot}
  }
  AssetSunsetVersion {
   widgetClassPath=&QuestDebugToolInfoField
   bIsVariable=true
   parent=HorizontalBox_163
   widget=AssetSunsetVersion
   properties {header=Asset Sunset}
   slot {class=HorizontalBoxSlot}
  }
  BuildWarning {
   widgetClassPath=&QuestDebugToolInfoField
   bIsVariable=true
   parent=VerticalBox_79
   widget=BuildWarning
   properties {displayAsAlert=true, header=Version Warning}
   slot {class=VerticalBoxSlot}
  }
 ]
}
