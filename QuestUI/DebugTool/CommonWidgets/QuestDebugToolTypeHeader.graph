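# Header widget for the quest debug tool that labels a DebugEntryType, such as schedule, bundle, quest or category views and assets.
# PreConstruct runs Populate, which sets the type text from the enum name and picks a description via a Select on the enum.
# Descriptions state whether a client view or instance exists, with fallback messages for unset or unsupported types.
# Populate also tints the rounded background using the shared Get Debug List Entry Type Color function.
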WidgetBlueprint /QuestUI/DebugTool/CommonWidgets/QuestDebugToolTypeHeader
graphs [
 EventGraph {nodes [
   Event PreConstruct()
   > Populate()
  ]}
 Populate {nodes [
   Populate() {is=title K2Node_FunctionEntry}
   n1 = Get DebugEntryType() -> (DebugEntryType:EQuestPanelDebugListEntryType Enum=Unset)
   > "SetText (Text)"(
    self=Get Text_Type() -> (Text_Type:Common Text Object Reference)
    InText="To Text (String)"(InString=Enum to String!(Enumerator:EQuestPanelDebugListEntryType Enum=@n1) -> (ReturnValue:String))
   )
   > "SetText (Text)"(
    self=Get Text_Desc() -> (Text_Desc:Common Text Object Reference)
    InText=Select!(
     Unset:Text="LOCTEXT(\"Please set the DebugEntryType\")"
     NotFound:Text="LOCTEXT(\"Unsupported\")"
     ScheduleView:Text="LOCTEXT(\"A view for this schedule exists on your client\")"
     ScheduleAsset:Text="LOCTEXT(\"These schedule asset(s) were found but no view exists on your client for them\")"
     BundleView:Text="LOCTEXT(\"A view for this bundle exists on your client\")"
     BundleAsset:Text="LOCTEXT(\"These bundle asset(s) were found but no view exists on your client for them\")"
     QuestView:Text="LOCTEXT(\"A view for this quest exists on your client\")"
     QuestAsset:Text="LOCTEXT(\"These quest asset(s) were found but no view exists on your client for them\")"
     CategoryInstance:Text="LOCTEXT(\"An instance for this category exists on your client\")"
     CategoryAsset:Text="LOCTEXT(\"This category asset was found but no instance exists on your client for it\")"
     Index:EQuestPanelDebugListEntryType Enum=Get DebugEntryType() -> (DebugEntryType=Unset)
    ) -> (ReturnValue:Text)
   )
   > SetColorAndOpacity(
    self:Image Object Reference=Get BG() -> (BG:Image Object Reference)
    InColorAndOpacity:Linear Color Structure=Get Debug List Entry Type Color!(EntryType:EQuestPanelDebugListEntryType Enum=Get DebugEntryType() -> (DebugEntryType=Unset)) -> (Color:Linear Color Structure) {is=title K2Node_CallFunction}
   )
  ]}
]
variables [DebugEntryType {type=EQuestPanelDebugListEntryType, default=Unset}]
widgets {
 info {inheritedWidgetCount=0, parentClass=&UserWidget, rootWidgetClass=&Overlay, widgetCount=5}
 widgets [
  Overlay_0 {widgetClassPath=&Overlay, widget=Overlay_0}
  BG {
   widgetClassPath=&Image
   bIsVariable=true
   parent=Overlay_0
   widget=BG
   properties {
    brush {
     bIsDynamicallyLoaded=false
     drawAs=RoundedBox
     imageType=NoImage
     mirroring=NoMirror
     resourceName=None
     resourceObject=null
     tiling=NoTile
     imageSize=<32 32>
     margin=<ltrb 0 0 0 0>
     outlineSettings {
      bUseBrushTransparency=false
      roundingType=HalfHeightRadius
      width=0
      color {colorUseRule=UseColor_Specified, specifiedColor=<rgba 0 0 0 0>}
      cornerRadii=<0 0 0 0>
     }
     tintColor {colorUseRule=UseColor_Specified, specifiedColor=<rgba 1 1 1 1>}
     uVRegion {bIsValid=false, max=<0 0>, min=<0 0>}
    }
    colorAndOpacity=<rgba 1 1 1 0.319>
   }
   slot {class=OverlaySlot, properties {horizontalAlignment=HAlign_Fill, verticalAlignment=VAlign_Fill, padding=<ltrb 5 5 5 5>}}
  }
  HorizontalBox_2 {
   widgetClassPath=&HorizontalBox
   parent=Overlay_0
   widget=HorizontalBox_2
   slot {class=OverlaySlot, properties {horizontalAlignment=HAlign_Fill, verticalAlignment=VAlign_Fill, padding=<ltrb 20 10 20 10>}}
  }
  Text_Type {
   widgetClassPath=&CommonTextBlock
   bIsVariable=true
   parent=HorizontalBox_2
   widget=Text_Type
   properties {
    style=/FortUILibrary/Styles/Typography/Foundation/TextStyle-Base-S-B.TextStyle-Base-S-B_C
    text=Type
    font {
     bForceMonospaced=false
     bMaterialIsStencil=false
     fontMaterial=null
     fontObject=&NotoSans
     letterSpacing=0
     monospacedWidth=1
     size=18
     skewAmount=0
     typefaceFontName=NotoSans-Bold
     outlineSettings {
      bApplyOutlineToDropShadows=false
      bMiteredCorners=false
      bSeparateFillAlpha=false
      outlineMaterial=null
      outlineSize=0
      outlineColor=<rgba 0 0 0 1>
     }
    }
    shadowOffset=<0 0>
   }
   slot {class=HorizontalBoxSlot, properties {verticalAlignment=VAlign_Center, padding=<ltrb 0 0 20 0>}}
  }
  Text_Desc {
   widgetClassPath=&CommonTextBlock
   bIsVariable=true
   parent=HorizontalBox_2
   widget=Text_Desc
   properties {
    style=/FortUILibrary/Styles/Typography/Foundation/TextStyle-Base-XXS-B.TextStyle-Base-XXS-B_C
    text=Description
    font {
     bForceMonospaced=false
     bMaterialIsStencil=false
     fontMaterial=null
     fontObject=&NotoSans
     letterSpacing=0
     monospacedWidth=1
     size=14
     skewAmount=0
     typefaceFontName=NotoSans-Bold
     outlineSettings {
      bApplyOutlineToDropShadows=false
      bMiteredCorners=false
      bSeparateFillAlpha=false
      outlineMaterial=null
      outlineSize=0
      outlineColor=<rgba 0 0 0 1>
     }
    }
    shadowOffset=<0 0>
   }
   slot {class=HorizontalBoxSlot, properties {horizontalAlignment=HAlign_Right, verticalAlignment=VAlign_Center, size {sizeRule=Fill, value=1}}}
  }
 ]
}
