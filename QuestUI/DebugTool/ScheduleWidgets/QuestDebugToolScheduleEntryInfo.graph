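# Quest debug tool widget that shows a "Schedule Entries" section listing schedule bundle entry debug data.
# A section header sits above a vertical DynamicEntryBox of QuestDebugToolScheduleEntry widgets.
# Set Entries resets the box, creates one entry widget per data structure in the given array, and passes that structure to the widget through SetData.
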WidgetBlueprint /QuestUI/DebugTool/ScheduleWidgets/QuestDebugToolScheduleEntryInfo
graphs [
 EventGraph {links [], nodes []}
 Set Entries {nodes [
   n1 = Set Entries!() -> (then, Entries:Array of Schedule Bundle Entry Debug Data Structures) {is=title K2Node_FunctionEntry}
   n2 = Get DynamicEntryBox_Entries() -> (DynamicEntryBox_Entries:Dynamic Entry Box Object Reference)
   > Reset!(execute, self:Dynamic Entry Box Object Reference=@n2, bDeleteWidgets:bool=false) -> (then)
   > n3 = For Each Loop(Array:Array of Schedule Bundle Entry Debug Data Structures=@n1) -> (Array Element:Schedule Bundle Entry Debug Data Structure)
   > n4 = Create Entry!(execute, self:Dynamic Entry Box Object Reference=@n2) -> (then, ReturnValue:User Widget Object Reference)
   > n5 = Cast To QuestDebugToolScheduleEntry!(execute, Object=@n4) -> (then, CastFailed, AsQuest Debug Tool Schedule Entry:Quest Debug Tool Schedule Entry Object Reference) {is=title K2Node_DynamicCast}
   > SetData!(execute, self:Quest Debug Tool Schedule Entry Object Reference=@n5, Entry:"Schedule Bundle Entry Debug Data Structure (by ref)"=@n3) -> (then)
  ]}
]
variables []
widgets {
 info {inheritedWidgetCount=0, parentClass=&UserWidget, rootWidgetClass=&VerticalBox, widgetCount=3}
 widgets [
  VerticalBox_28 {widgetClassPath=&VerticalBox, widget=VerticalBox_28}
  SectionHeader {
   widgetClassPath=/QuestUI/DebugTool/CommonWidgets/QuestDebugToolSectionHeader
   bIsVariable=true
   parent=VerticalBox_28
   widget=SectionHeader
   properties {header=Schedule Entries}
   slot {class=VerticalBoxSlot}
  }
  DynamicEntryBox_Entries {
   widgetClassPath=&DynamicEntryBox
   bIsVariable=true
   parent=VerticalBox_28
   widget=DynamicEntryBox_Entries
   properties {
    entryBoxType=Vertical
    entryWidgetClass=&/QuestDebugToolScheduleEntry.QuestDebugToolScheduleEntry_C
    entrySpacing=<0 2>
   }
   slot {class=VerticalBoxSlot, properties {padding=<ltrb 6 0 6 6>}}
  }
 ]
}
