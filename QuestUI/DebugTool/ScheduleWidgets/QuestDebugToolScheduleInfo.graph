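# Widget Blueprint for the schedule info panel in the quest debug tool UI.
# A vertical box stacks a shared section header titled Schedule Info above a reusable definition field widget.
# The definition field is configured for schedule asset and view entry types, with headers for the schedule definition and its display data.
# The event graph is empty and no variables are declared, so the panel itself is layout and property configuration only.
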WidgetBlueprint /QuestUI/DebugTool/ScheduleWidgets/QuestDebugToolScheduleInfo
graphs [EventGraph {links [], nodes []}]
variables []
widgets {
 info {inheritedWidgetCount=0, parentClass=&UserWidget, rootWidgetClass=&VerticalBox, widgetCount=3}
 widgets [
  VerticalBox_1 {widgetClassPath=&VerticalBox, widget=VerticalBox_1}
  SectionHeader {
   widgetClassPath=/QuestUI/DebugTool/CommonWidgets/QuestDebugToolSectionHeader
   bIsVariable=true
   parent=VerticalBox_1
   widget=SectionHeader
   properties {header=Schedule Info}
   slot {class=VerticalBoxSlot}
  }
  Widget_ScheduleDefinition {
   widgetClassPath=/QuestUI/DebugTool/CommonWidgets/QuestDebugToolDefinitionField
   bIsVariable=true
   parent=VerticalBox_1
   widget=Widget_ScheduleDefinition
   properties {assetEntryType=ScheduleAsset, definitionHeader=Schedule Definition, displayDataHeader=Schedule Display Data, viewEntryType=ScheduleView}
   slot {class=VerticalBoxSlot}
  }
 ]
}
